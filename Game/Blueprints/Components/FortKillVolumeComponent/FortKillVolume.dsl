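; Kill volume component that applies environment damage to actors overlapping it.
; On BeginPlay it binds its begin-overlap event, skips actors matching ignore gameplay tags, and applies global environment damage using a scalable float amount and a gameplay effect.
; A separate event disables collision on the owner's other mesh components, printing each mesh component's name to screen as debug output.
; The visibility-enable event is an empty stub.
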(event EventBeginPlay
  (Collision|BindEventtoOnComponentBeginOverlap (delegate Custom|Event%20Apply%20Damage%20from%20KV)))

(event Custom|Event%20Apply%20Damage%20from%20KV
  (params (OverlappedComponent :type Primitive%20Component%20Object%20Reference) (OtherActor :type Actor%20Object%20Reference) (OtherComp :type Primitive%20Component%20Object%20Reference) (OtherBodyIndex :type Integer) (bFromSweep :type Boolean) (SweepResult :type Hit%20Result%20Structure%20%28by%20ref%29))
  (bind _returnvalue (GameplayTags|HasAnyMatchingGameplayTags(Message) OtherActor (Variables|Default|GetTC_IgnoreActorsWithTags)))
  (if (not _returnvalue)
    (Environment|ApplyGlobalEnvironmentDamagetoActor :DamageTargetActor OtherActor :EnvironmentDamage (ScalableFloat|GetValueatLevel (Variables|Default|GetEnvironment_Damage)) :EnvironmentDamageGE (Variables|Default|GetEnvironment%20Damage%20GE))))

(event Custom|KVComp_DisableCollision
  (for _array_element (Actor|GetComponentsByClass (Components|GetOwner) "/Script/Engine.MeshComponent")
    (Development|PrintString (Utilities|GetDisplayName _array_element))
    (if (!= _array_element self)
      (Collision|SetCollisionEnabled _array_element "NoCollision"))))

(event Custom|KVComp_EnableVisibility)
